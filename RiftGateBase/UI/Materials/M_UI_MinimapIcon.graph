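# Translucent UI material for a minimap icon, defaulting to a portal texture, with animated pulse rings tinted by PulseColor.
# A center-scaled icon texture is composited over the tint, using its green channel as a mask.
# Two MF_CirclePulse rings, clipped by a circle, combine via Max with time-animated center dots and the icon mask.
# A separate branch offers a static switch for SmoothStep edges on SDF textures.
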Material /RiftGateBase/UI/Materials/M_UI_MinimapIcon
graph {
 properties {blendMode=BLEND_Translucent, materialDomain=MD_UI, pixelDepthOffsetMode=PDOM_Legacy, usedShadingModels=Default Lit}
 nodes [
  n1 = VectorParameter() {parameterName=PulseColor, defaultValue=<rgba 0.8308 0 0.0222 1>}
  n2 = Add(A=@n1.RGB) {constB=0.1}
  n3 = Multiply(A=MaterialFunctionCall(Size=Constant() {r=0.05}, GlowMax=Constant() {r=0.5}) {materialFunction=&MF_UI_Circle}.Fill) {constB=0.5}
  n4 = TextureSampleParameter2D(UVs=MaterialFunctionCall(UVs=TextureCoordinate(), Texture Scale=ScalarParameter() {defaultValue=1, group=Icon, parameterName=IconScale}) {materialFunction=&ScaleUVsByCenter}.UVs) {group=Icon, parameterName=IconTexture, samplerSource=SSM_Clamp_WorldGroupSettings, texture=/RiftGateBase/UI/Textures/T_UI_MinimapIcon_Portal}
  n5 = Add(A=@n4.G, B=@n4.G) {constB=0}
  n6 = Saturate(None=@n5)
  Add(A=Multiply(A=LinearInterpolate(A=@n1.RGB, B=@n2, Alpha=@n3), B=OneMinus(None=@n6)), B=Min(A=@n5, B=@n4.RGB))
  Constant() {r=1.25}
  n7 = ScalarParameter() {defaultValue=0.2, group=Icon, parameterName=Edge}
  n8 = ScalarParameter() {defaultValue=0.1, group=Icon, parameterName=EdgeSoftness}
  n9 = MaterialFunctionCall(Alpha=@n4.R, Min=@n7, Max=Add(A=@n7, B=Multiply(A=@n7, B=@n8) {constB=0.1})) {materialFunction=&SmoothStep}
  LinearInterpolate(B=@n4.RGB, Alpha=Max(A=@n4.R, B=StaticSwitchParameter(True=@n9.Result, False=@n4.R) {group=Icon, parameterName=IsSDFTexture}))
  n10 = ScalarParameter() {group=PulseStyle, parameterName="ScaleStart (S)"}
  n11 = Constant() {r=0.25}
  n12 = Max(
   A=MaterialFunctionCall!(
    PulseSpeed
    SizeEnd=ScalarParameter() {defaultValue=1.1, group=PulseStyle, parameterName=ScaleEnd}
    WidthStart
    WidthEnd=ScalarParameter() {defaultValue=0.2, group=PulseStyle, parameterName="WidthEnd (S)"}
    PulseAnimFalloff
    SizeStart=@n10
    OpacityStart
    OpacityEnd=@n11
    TimeSpeed
    TexCoord
   ) {materialFunction=&/MF_CirclePulse}.Result
   B=MaterialFunctionCall(
    SizeEnd=ScalarParameter() {defaultValue=1, group=PulseStyle, parameterName=ScaleEnd2}
    WidthEnd=ScalarParameter() {defaultValue=0.01, group=PulseStyle, parameterName=WidthEnd2}
    PulseAnimFalloff=ScalarParameter() {defaultValue=0.2605, group=PulseStyle, parameterName="PulseAnimFalloff (S)"}
    SizeStart=@n10
    OpacityEnd=@n11
   ) {materialFunction=&MF_CirclePulse}.Result
  )
  n13 = MaterialFunctionCall(Size=Constant() {r=0.5}, GlowMax=Constant() {r=0.5}) {materialFunction=&MF_UI_Circle}
  n14 = Constant() {r=0.08}
  n15 = Time()
  n16 = Multiply(A=Cosine(None=Frac(None=@n15))) {constB=0.1}
  n17 = Constant() {r=0.2}
  n18 = Add(
   A=MaterialFunctionCall(Size=Add(A=@n14, B=@n16), GlowMax=Add(A=@n16, B=@n17)) {materialFunction=&MF_UI_Circle}.Fill
   B=MaterialFunctionCall(Size=Constant() {r=0.08}, GlowMax=Constant() {r=0.2}) {materialFunction=&MF_UI_Circle}.Fill
  )
  Max(A=@n4.G, B=Max(A=Min(A=@n12, B=@n13.Fill), B=Multiply(A=@n18) {constB=0.5}))
 ]
}
